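# Translucent, lit material for dust motes, tinted by the FortniteActiveTimeOfDay parameter, which blends morning, day, evening and night colors into a boosted diffuse color.
# Pixel depth crossfades a blurred flare texture near the camera into a toon flare farther away.
# The graph modulates that mask by particle alpha or by a sine pulse randomized per Pivot Painter element.
# View-space offset math, including a variant around Pivot Painter pivots, builds camera-facing quads.
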Material /Game/Effects/Fort_Effects/World/Enviroment/Dust/M_Dust_Particles
graph {
 properties {
  blendMode=BLEND_Translucent
  pixelDepthOffsetMode=PDOM_Legacy
  translucencyLightingMode=TLM_VolumetricPerVertexNonDirectional
  usedShadingModels=Default Lit
 }
 nodes [
  Constant() {r=1}
  n1 = CollectionParameter() {
   collection=&FortniteMaterialParameters
   parameterName=FortniteActiveTimeOfDay
  }
  n2 = Constant4Vector() {constant=<rgba 0 0 0 1>}
  n3 = LinearInterpolate(
   A=LinearInterpolate(
    A=Constant3Vector() {desc=Morning, constant=<rgba 1 0.5297 0.1637 0>}
    B=Constant3Vector() {desc=Night, constant=<rgba 0.2362 0.4407 1 0>}
    Alpha=DotProduct(A=@n1, B=@n2.RGBA)
   )
   B=Constant3Vector() {desc=Evening, constant=<rgba 1 0.3764 0.0343 0>}
   Alpha=DotProduct(A=@n1, B=Constant4Vector() {constant=<rgba 0 0 1 0>}.RGBA)
  )
  n4 = Constant3Vector() {desc=Day, constant=<rgba 1 0.6439 0.1455 0>}
  n5 = DotProduct(A=@n1, B=Constant4Vector() {constant=<rgba 0 1 0 0>}.RGBA)
  MaterialFunctionCall(DiffuseColor=Multiply(A=LinearInterpolate(A=@n3, B=@n4, Alpha=@n5), B=Constant() {r=3})) {materialFunction=&ConvertFromDiffSpec}
  n6 = Subtract(A=PixelDepth().R, B=Constant() {desc=Start Blend Dist, r=50})
  n7 = Constant() {desc=Blend Dist, r=50}
  n8 = LinearInterpolate(
   A=TextureSample() {texture=/Game/Effects/Fort_Effects/Textures/Flares/T_Toon_Flare_Blur}.RGB
   B=TextureSample() {texture=/Game/Effects/Fort_Effects/Textures/Flares/T_Toon_Flare1}.RGB
   Alpha=Clamp(None=Divide(A=@n6, B=@n7))
  )
  n9 = MaterialFunctionCall() {materialFunction=/Engine/Functions/Engine_MaterialFunctions02/PivotPainter/PivotPainter_PerObjectData}
  n10 = Time()
  n11 = MaterialFunctionCall(Value=Add(A=@n9.Random Value Per Element, B=Multiply(A=@n10) {constB=0.5})) {materialFunction=&LinearSine}
  Multiply(A=@n8, B=Clamp(None=Subtract(A=@n11.Linear Sine) {constB=0.4}))
  Multiply(A=ParticleColor().A, B=@n8)
  n12 = Transform(None=Constant3Vector() {constant=<rgba 0 1 0 0>}) {transformSourceType=TRANSFORMSOURCE_View}
  n13 = Transform(None=Constant3Vector() {constant=<rgba 1 0 0 0>}) {transformSourceType=TRANSFORMSOURCE_View}
  Subtract(A=Multiply(A=Add(A=@n12, B=@n13), B=Constant() {desc=Size, r=50}), B=WorldPosition().XYZ)
  n14 = AppendVector(A=Subtract(A=TextureCoordinate()) {constB=0.5}, B=Constant())
  n15 = Constant3Vector() {constant=<rgba -1 1 1 0>}
  n16 = Normalize(VectorInput=Transform(None=Multiply(A=@n14, B=@n15)) {transformSourceType=TRANSFORMSOURCE_View})
  n17 = Constant() {desc=Size, r=1.6}
  Subtract(A=Add(A=@n9.Pivot Position, B=Multiply(A=@n16, B=@n17)), B=WorldPosition().XYZ)
 ]
}
